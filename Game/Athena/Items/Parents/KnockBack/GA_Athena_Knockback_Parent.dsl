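; Parent Gameplay Ability that knocks its avatar away from or toward a launcher, optionally applying a knockback status effect.
; Activation filters by tags and team, then builds velocity from scalable magnitudes, optional distance falloff and retained target velocity.
; Pawns get gravity-compensated constant or move-to root motion forces, plus fall-instigator and bot handling.
; Non-pawns fall back to vehicle launches via an impact broadcast, or impulses on tagged physics building actors.
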(event Custom|LaunchBuldingActors
  (params (Actor :type Actor%20Object%20Reference))
  ;[1 1 1] If the actor has the building actor Physics tag and also a physics component launch it.
  (FortPhysicsSimulationLibrary|SetImpactInstigatorforPhysicsObject Actor (Variables|KnockbackParent|GetLauncher))
  (bind _returnvalue (GameplayTags|HasMatchingGameplayTag(Message) Actor (Variables|KnockbackParent|GetTag_BuildingPhysics)))
  (if _returnvalue
    (Utilities|IsValid (Actor|GetComponentbyClass Actor "/Script/FortniteGame.FortPhysicsObjectComponent")
      (:Is%20Valid
        (if (Collision|GetActorEnableCollision Actor)
          (bind _asprimitive_component (Utilities|Casting|CastToPrimitiveComponent (Variables|/Script/Engine.Actor|GetRootComponent Actor))
            (:then
              (Physics|AddImpulseatLocation _asprimitive_component (* (Variables|KnockbackParent|GetFinalLaunchVel) (ScalableFloat|GetValueatLevel (Variables|KnockbackParent|GetRow_PhysicsObjectImpulseMultiplier))) (Transformation|GetActorLocation Actor) "None")
              (Utilities|FlowControl|Delay 0.2))
            (:CastFailed
              _))
          (Ability|EndAbility)
          (else
            _)))
      (:Is%20Not%20Valid
        _))
    (else
      _)))

(event Ability|EventActivateAbility
  (bind _launcher (Variables|KnockbackParent|GetLauncher))
  (bind _returnvalue (CallFunction|GetKnockbackOrigin _launcher))
  (bind _launcher_2 (Variables|KnockbackParent|GetLauncher))
  (bind _returnvalue_2 (Fortnite|GetControllerfromActor _launcher_2))
  (bind _returnvalue_3 (Actor|GetOwner _launcher_2))
  (bind _returnvalue_4 (Utilities|IsValid _returnvalue_3 :!class /Script/Engine.KismetSystemLibrary))
  (bind _returnvalue_5 (Fortnite|GetControllerfromActor _returnvalue_3))
  (bind _returnvalue_6 (select _returnvalue_4 _returnvalue_5 _returnvalue_2))
  ;[1 1 1] Cast to pawn and launch the player a little bit. The distance launched is based on a max distance from the player who hit them. If the player is further than the max distance than no launch is applied. Otherwise the player is launched based on a ratio of the knock mag using the distance from the player who launched them.
  (Utilities|IsValid (Ability|GetAvatarActorfromActorInfo)
    (:Is%20Valid
      (bind _launcherout (CallFunction|FindLauncher (Default|GetAbilityInstigator))) ; allow atypical setups to find the launcher \nbefore checking if it is valid
      (bind _output_get (Variables|KnockbackParent|SetLauncher _launcherout))
      (Utilities|IsValid _output_get
        (:Is%20Valid
          (if (Variables|KnockbackParent|GetShouldApplyKnockbackStatus)
            (bind _returnvalue_7 (Ability|GameplayEffect|SetDuration (Ability|MakeOutgoingGameplayEffectSpec (Variables|Default|GetGE_Knockback_Status)) (Math|Float|Max(Float) (select (Variables|KnockbackParent|GetSetKnockbackDurationEqualToLaunchDuration) (Variables|KnockbackParent|GetLaunchDuration) (ScalableFloat|GetValueatLevel (Variables|KnockbackParent|GetRow_KnockbackDuration))))))
            (Ability|ApplyGameplayEffectSpecToOwner _returnvalue_7)
            (else
              _))
          (bind _output_get_2 (Variables|KnockbackParent|SetTarget (Ability|GetAvatarActorfromActorInfo)))
          ;;[1 1 1] Check to ensure we do not have tags that should stop the launch
          (bind _returnvalue_8 (GameplayTags|HasAnyMatchingGameplayTags(Message) _output_get_2 (Variables|KnockbackParent|GetTC_TagsToIgnore)))
          (if (not _returnvalue_8)
            (bind _doknockback (CallFunction|CheckIgnoreKnockifonSameTeam))
            (if _doknockback
              ;;[1 1 1] Calc Distance to apply knockback harder if closer
              (if (or (Variables|KnockbackParent|GetXYMagUsesInstigatorDistanceMod) (Variables|KnockbackParent|GetZMagUsesInstigatorDistanceMod))
                (bind (_modamount _abovemax) (CallFunction|CalculateDistanceMod (Transformation|GetActorLocation _output_get_2) _returnvalue))
                (if _abovemax
                  (Ability|EndAbility)
                  (else
                    _))
                (else
                  _))
              (bind _output_get_3 (Variables|KnockbackParent|SetFinalLaunchVel (CallFunction|CalculateVelocityMaintained (Variables|KnockbackParent|GetTarget) (CallFunction|GetInitialKnockMag (select (Variables|KnockbackParent|GetKnockTowardsLauncher) (- _returnvalue (Transformation|GetActorLocation _output_get_2)) (- (Transformation|GetActorLocation _output_get_2) _returnvalue))))))
              (bind _asfort_pawn (Utilities|Casting|CastToFortPawn (Variables|KnockbackParent|GetTarget))
                (:then
                  ;[1 1 1] Pawn Launch
                  (EventDispatchers|BindEventtoOnDied _asfort_pawn (EventDispatchers|CreateEvent :!event Custom|OwningPawnDied))
                  ;;[1 1 1] Stop Movement and launch them
                  (CallFunction|DoLaunch _output_get_3)
                  ;;[1 1 1] Set Fall Instigator
                  (bind _asfort_player_pawn_athena (Utilities|Casting|CastToFortPlayerPawnAthena _asfort_pawn)
                    (:then
                      (Utilities|IsValid _returnvalue_6
                        (:Is%20Valid
                          (Class|FortPlayerPawnAthena|SetFallInstigator _asfort_player_pawn_athena _returnvalue_6 (Variables|KnockbackParent|GetDefaultTags)))
                        (:Is%20Not%20Valid
                          _))
                      (if (Variables|KnockbackParent|GetResetFallHeight)
                        (if (or (and (> (.z _output_get_3) 0.0) (Variables|KnockbackParent|GetImpulseMustBePositiveForReset)) (not (Variables|KnockbackParent|GetImpulseMustBePositiveForReset)))
                          (Pawn|Character|ResetFallingHeight _asfort_pawn)
                          ;;[1 1 1] If target is a bot, notify it of being knockbacked to check if movement/reaction are affected
                          (if (Pawn|IsBotControlled _asfort_pawn)
                            (bind _asfort_athena_aibot_controller (Utilities|Casting|CastToFortAthenaAIBotController (Pawn|GetController _asfort_pawn))
                              (:then
                                (Class|FortAthenaAIBotController|OnKnockbacked _asfort_athena_aibot_controller (Variables|KnockbackParent|GetKnockbackTypeTag)))
                              (:CastFailed
                                _))
                            (Utilities|Time|SetTimerbyEvent (EventDispatchers|CreateEvent :!event Custom|FailsafeEndAbility) (Variables|Default|GetFailsafeDelay))))))
                    (:CastFailed
                      _)))
                (:CastFailed
                  ;[1 1 1] Vehicle Launch
                  (if (or (Fortnite|IsPhysicallySimulatedIsland) (Variables|KnockbackParent|Vehicle|GetKnocksVehicles))
                    (bind _asfort_vehicle_interface (Utilities|Casting|CastToFortVehicleInterface (Variables|KnockbackParent|GetTarget))
                      (:then
                        (CallFunction|LaunchVehicle _asfort_vehicle_interface (* (Math|Vector|Normalize (CallFunction|CalcVehicleMult _output_get_3)) (* (CallFunction|CheckGravityScale "(AttributeName=\"VehicleGravityScale\",Attribute=/Script/FortniteGame.FortVehicleMovementSet:VehicleGravityScale,AttributeOwner=None)") (Math|Vector|VectorLength (CallFunction|CalcVehicleMult _output_get_3)))))
                        (Utilities|FlowControl|Delay 0.2)
                        (Ability|EndAbility))
                      (:CastFailed
                        (CallFunction|LaunchBuldingActors (Variables|KnockbackParent|GetTarget)) ; Handle Launching Building Actors if It does not fire off from pawn or vehicle
                      ))))) ; Check if pawn or not
              (else
                _))
            (else
              _))
          (Ability|EndAbility))
        (:Is%20Not%20Valid
          (Development|PrintString "KNOCKBACK: Launcher Not Valid" true true "(R=1.000000,G=0.060296,B=0.000000,A=1.000000)" 2e+01))))
    (:Is%20Not%20Valid
      (Development|PrintString "LaunchTargetNotValid")))
  (Ability|EndAbility))

(event Custom|FailsafeEndAbility
  ;[1 1 1] Pawn Launch
  (Ability|EndAbility))

(event Custom|OwningPawnDied
  (params (DamagedActor :type Actor%20Object%20Reference) (Damage :type Float%20%28single-precision%29) (InstigatedBy :type Controller%20Object%20Reference) (DamageCauser :type Actor%20Object%20Reference) (HitLocation :type Vector) (FHitComponent :type Primitive%20Component%20Object%20Reference) (BoneName :type Name) (Momentum :type Vector))
  ;[1 1 1] End ability if pawn has died
  (Development|PrintString "Pawn is dead Remove ability")
  (Ability|EndAbility))

(event Custom|DoLaunch
  (params (Vel :type Vector))
  ;[1 1 1] Do Launch - Do the Launch of the character based on the launch type chosen.
  (switch Utilities|FlowControl|Switch|SwitchonAthena_Knockback_Parent_Type (Variables|KnockbackParent|GetLaunchType)
    (:NewEnumerator0
      (Ability|Tasks|ApplyRootMotionConstantForce "None" (Math|Vector|Normalize Vel) (* (CallFunction|CheckGravityScale "(AttributeName=\"GravityZScale\",Attribute=/Script/FortniteGame.FortMovementSet:GravityZScale,AttributeOwner=None)") (Math|Vector|VectorLength Vel)) (Variables|KnockbackParent|GetLaunchDuration) false (Variables|KnockbackParent|GetCurve_StrengthOverTime) (Variables|KnockbackParent|GetVelocityOnFinish) (Variables|KnockbackParent|GetSetVelocityOnFinish) (Variables|KnockbackParent|GetClampVelocityOnFinish) (Variables|KnockbackParent|GetUsesGravity)
        (:OnFinish
          _)))
    (:NewEnumerator3
      (bind _returnvalue (CallFunction|GetTargetLocation))
      (Ability|Tasks|ApplyRootMotionMoveToForce "None" _returnvalue (Variables|KnockbackParent|GetLaunchDuration) false "MOVE_None" false (Variables|KnockbackParent|GetCurve_PathOffset) (Variables|KnockbackParent|GetVelocityOnFinish) (Variables|KnockbackParent|GetSetVelocityOnFinish) (Variables|KnockbackParent|GetClampVelocityOnFinish)
        (:OnTimedOutAndDestinationReached
          _))))
  (CallFunction|LaunchFinished))

(event Custom|LaunchFinished
  ;[1 1 1] Launch Finished - By default on launch finished end ability
  (Utilities|FlowControl|Delay 0.2) ; Small Delay to make sure the ability has time to clean anything
  (Ability|EndAbility) ; On Finished End Ability
)

(event Ability|EventFailedToActivatePassiveAbility
  (params (ActorInfo :type Gameplay%20Ability%20Actor%20Info%20Structure))
  ;[1 1 1] Failed to activate a passive ability
  (|CallParentFunction ActorInfo))

(fn CalculateDistanceMod
  (params (TargetLocation :type Vector) (KnockOrigin :type Vector))
  (bind _row_maxdistance (Variables|KnockbackParent|GetRow_MaxDistance))
  (bind _returnvalue (ScalableFloat|GetValueatLevel _row_maxdistance))
  ;[0.3692 0.6821 0.8646] CalculateDistanceMod
  ; Use a max distance to determine the falloff of the knockback magnitude
  ; a minimum value prevents all falloff within its range
  (bind _distance (Class|MissionBlueprintFunctionLibrary|DistanceBetweenTwoVectors TargetLocation KnockOrigin))
  (bind _output_get (Variables|KnockbackParent|SetDistance _distance))
  (if (< _output_get _returnvalue)
    (if (> _output_get (ScalableFloat|GetValueatLevel (Variables|KnockbackParent|GetRow_MinDistance)))
      (bind _output_get_2 (Variables|KnockbackParent|SetDistanceMagMult (Math|Float|Absolute(Float) (- (Math|Float|SafeDivide _output_get _returnvalue) 1.0))))
      ;;[1 0.5654 0.8843] RETURN
      ;; Knockback magnitude has been modified and approaches zero as
      (return _output_get_2 false)
      (else
        ;;[1 0.5654 0.8843] RETURN
        ;; Within the knockback range, but is within the minumum for the magnitude scaling, the knockback mod value is always unmodified at 100% (1.0)
        (return 1.0 false)))
    (else
      ;;[1 0.5654 0.8843] RETURN
      ;; Above Max is TRUE
      ;; allows knockback to be disabled
      (return 0.0 true))))

(fn CalculateVelocityMaintained
  (params (LaunchTarget :type Actor%20Object%20Reference) (LaunchMagIn :type Vector))
  ;[1 1 1] Calculate How much to modify launch speed by from the actors given velocity.
  (return (+ LaunchMagIn (Math|Vector|MakeVector (* (.x (Transformation|GetVelocity LaunchTarget)) (ScalableFloat|GetValueatLevel (Variables|KnockbackParent|GetRow_XYVelPercentMaintained))) (* (.y (Transformation|GetVelocity LaunchTarget)) (ScalableFloat|GetValueatLevel (Variables|KnockbackParent|GetRow_XYVelPercentMaintained))) (* (.z (Transformation|GetVelocity LaunchTarget)) (ScalableFloat|GetValueatLevel (Variables|KnockbackParent|GetRow_ZVelPercentMaintained)))))))

(fn CalcVehicleMult
  (params (InVec :type Vector))
  ;[1 1 1] Calculate the impulse amount the vehicle should get.
  (return (Math|Vector|MakeVector (* (.x InVec) (ScalableFloat|GetValueatLevel (Variables|KnockbackParent|Vehicle|GetRow_VehicleXYMult))) (* (.y InVec) (ScalableFloat|GetValueatLevel (Variables|KnockbackParent|Vehicle|GetRow_VehicleXYMult))) (* (.z InVec) (ScalableFloat|GetValueatLevel (Variables|KnockbackParent|Vehicle|GetRow_VehicleZMult))))))

(fn CheckGravityScale
  (params (Attribute%20Gravity%20Scale :type Gameplay%20Attribute%20Structure))
  (bind _target (Variables|KnockbackParent|GetTarget))
  (bind _returnvalue (Actor|GetComponentbyClass _target "/Script/Engine.CharacterMovementComponent"))
  (bind (_bsuccessfullyfoundattribute _returnvalue_2) (Ability|Attribute|GetFloatAttribute _target Attribute%20Gravity%20Scale))
  (bind _localgravityscale (Variables|Default|GetLocalGravityScale))
  ;[1 1 1] Get the gravity scale Attribute and calc the gravity modifier. If it is more than 1.0 then try to increase the power and if it is less than 1.0 decrease the power
  (if (Variables|KnockbackParent|GetAccountsForGravityZScale)
    (if _bsuccessfullyfoundattribute
      (if (and (== _returnvalue_2 1.0) (Utilities|IsValid _returnvalue :!class /Script/Engine.KismetSystemLibrary))
        (Variables|Default|SetLocalGravityScale (Variables|/Script/Engine.CharacterMovementComponent|GetGravityScale _returnvalue))
        (else
          (Variables|Default|SetLocalGravityScale _returnvalue_2)))
      (if (< _localgravityscale 1.0)
        (return (Math|Float|SafeDivide _localgravityscale (Variables|KnockbackParent|GetLowZGravityDampenerIntensity)))
        (elif (> _localgravityscale 1.0)
          (return (* _localgravityscale (Variables|KnockbackParent|GetHighZGravityMultiplierIntensity)))))
      (else
        _))
    (return 1.0)
    (else
      _)))

(fn CheckIgnoreKnockIfOnSameTeam
  ; Do we want to ignore the Knockback if the Launcher and the Affected Player are on the same team?
  (if (ScalableFloat|GetValueasBool (Variables|KnockbackParent|GetAllowImpulsingTeammates))
    (return true)
    (else
      (return (not (Team|OnSameTeam (Variables|KnockbackParent|GetLauncher) (Variables|KnockbackParent|GetTarget)))))))

(fn FindLauncher
  (params (DefaultLauncher :type Actor%20Object%20Reference))
  ;[1 1 1] Function for finding the launcher. By default is the effect causer but can be overriden to do logic
  (return DefaultLauncher))

(fn GetInitialKnockMag
  (params (Direction :type Vector) (OverrideXYMag :type Boolean) (XYMAgOverride :type Float%20%28double-precision%29))
  (bind (_returnvalue_x _returnvalue_y _returnvalue_z) (Math|Vector|BreakVector (Math|Vector|Normalize Direction)))
  ;[1 1 1] Calculate the Launch magnitude from the Launch amounts and potentially the distance from the target if desired.
  (return (Math|Vector|MakeVector (.x (Math|Vector|MakeVector (* _returnvalue_x (Math|Float|SelectFloat (* (Math|Float|SelectFloat XYMAgOverride (ScalableFloat|GetValueatLevelwithContextParam (Variables|KnockbackParent|GetLauncher) (Variables|KnockbackParent|GetRow_XYKnockMag)) OverrideXYMag) (Variables|KnockbackParent|GetDistanceMagMult)) (Math|Float|SelectFloat XYMAgOverride (ScalableFloat|GetValueatLevelwithContextParam (Variables|KnockbackParent|GetLauncher) (Variables|KnockbackParent|GetRow_XYKnockMag)) OverrideXYMag) (Variables|KnockbackParent|GetXYMagUsesInstigatorDistanceMod))) (* _returnvalue_y (Math|Float|SelectFloat (* (Math|Float|SelectFloat XYMAgOverride (ScalableFloat|GetValueatLevelwithContextParam (Variables|KnockbackParent|GetLauncher) (Variables|KnockbackParent|GetRow_XYKnockMag)) OverrideXYMag) (Variables|KnockbackParent|GetDistanceMagMult)) (Math|Float|SelectFloat XYMAgOverride (ScalableFloat|GetValueatLevelwithContextParam (Variables|KnockbackParent|GetLauncher) (Variables|KnockbackParent|GetRow_XYKnockMag)) OverrideXYMag) (Variables|KnockbackParent|GetXYMagUsesInstigatorDistanceMod))) (* (Math|Float|SelectFloat (Math|Float|Clamp(Float) _returnvalue_z) _returnvalue_z (Variables|KnockbackParent|GetZAlwayzPositive)) (Math|Float|SelectFloat (* (ScalableFloat|GetValueatLevelwithContextParam (Variables|KnockbackParent|GetLauncher) (Variables|KnockbackParent|GetRow_ZKnockMag)) (Variables|KnockbackParent|GetDistanceMagMult)) (ScalableFloat|GetValueatLevelwithContextParam (Variables|KnockbackParent|GetLauncher) (Variables|KnockbackParent|GetRow_ZKnockMag)) (Variables|KnockbackParent|GetZMagUsesInstigatorDistanceMod))))) (.y (Math|Vector|MakeVector (* _returnvalue_x (Math|Float|SelectFloat (* (Math|Float|SelectFloat XYMAgOverride (ScalableFloat|GetValueatLevelwithContextParam (Variables|KnockbackParent|GetLauncher) (Variables|KnockbackParent|GetRow_XYKnockMag)) OverrideXYMag) (Variables|KnockbackParent|GetDistanceMagMult)) (Math|Float|SelectFloat XYMAgOverride (ScalableFloat|GetValueatLevelwithContextParam (Variables|KnockbackParent|GetLauncher) (Variables|KnockbackParent|GetRow_XYKnockMag)) OverrideXYMag) (Variables|KnockbackParent|GetXYMagUsesInstigatorDistanceMod))) (* _returnvalue_y (Math|Float|SelectFloat (* (Math|Float|SelectFloat XYMAgOverride (ScalableFloat|GetValueatLevelwithContextParam (Variables|KnockbackParent|GetLauncher) (Variables|KnockbackParent|GetRow_XYKnockMag)) OverrideXYMag) (Variables|KnockbackParent|GetDistanceMagMult)) (Math|Float|SelectFloat XYMAgOverride (ScalableFloat|GetValueatLevelwithContextParam (Variables|KnockbackParent|GetLauncher) (Variables|KnockbackParent|GetRow_XYKnockMag)) OverrideXYMag) (Variables|KnockbackParent|GetXYMagUsesInstigatorDistanceMod))) (* (Math|Float|SelectFloat (Math|Float|Clamp(Float) _returnvalue_z) _returnvalue_z (Variables|KnockbackParent|GetZAlwayzPositive)) (Math|Float|SelectFloat (* (ScalableFloat|GetValueatLevelwithContextParam (Variables|KnockbackParent|GetLauncher) (Variables|KnockbackParent|GetRow_ZKnockMag)) (Variables|KnockbackParent|GetDistanceMagMult)) (ScalableFloat|GetValueatLevelwithContextParam (Variables|KnockbackParent|GetLauncher) (Variables|KnockbackParent|GetRow_ZKnockMag)) (Variables|KnockbackParent|GetZMagUsesInstigatorDistanceMod))))) (+ (.z (Math|Vector|MakeVector (* _returnvalue_x (Math|Float|SelectFloat (* (Math|Float|SelectFloat XYMAgOverride (ScalableFloat|GetValueatLevelwithContextParam (Variables|KnockbackParent|GetLauncher) (Variables|KnockbackParent|GetRow_XYKnockMag)) OverrideXYMag) (Variables|KnockbackParent|GetDistanceMagMult)) (Math|Float|SelectFloat XYMAgOverride (ScalableFloat|GetValueatLevelwithContextParam (Variables|KnockbackParent|GetLauncher) (Variables|KnockbackParent|GetRow_XYKnockMag)) OverrideXYMag) (Variables|KnockbackParent|GetXYMagUsesInstigatorDistanceMod))) (* _returnvalue_y (Math|Float|SelectFloat (* (Math|Float|SelectFloat XYMAgOverride (ScalableFloat|GetValueatLevelwithContextParam (Variables|KnockbackParent|GetLauncher) (Variables|KnockbackParent|GetRow_XYKnockMag)) OverrideXYMag) (Variables|KnockbackParent|GetDistanceMagMult)) (Math|Float|SelectFloat XYMAgOverride (ScalableFloat|GetValueatLevelwithContextParam (Variables|KnockbackParent|GetLauncher) (Variables|KnockbackParent|GetRow_XYKnockMag)) OverrideXYMag) (Variables|KnockbackParent|GetXYMagUsesInstigatorDistanceMod))) (* (Math|Float|SelectFloat (Math|Float|Clamp(Float) _returnvalue_z) _returnvalue_z (Variables|KnockbackParent|GetZAlwayzPositive)) (Math|Float|SelectFloat (* (ScalableFloat|GetValueatLevelwithContextParam (Variables|KnockbackParent|GetLauncher) (Variables|KnockbackParent|GetRow_ZKnockMag)) (Variables|KnockbackParent|GetDistanceMagMult)) (ScalableFloat|GetValueatLevelwithContextParam (Variables|KnockbackParent|GetLauncher) (Variables|KnockbackParent|GetRow_ZKnockMag)) (Variables|KnockbackParent|GetZMagUsesInstigatorDistanceMod))))) (ScalableFloat|GetValueatLevelwithContextParam (Variables|KnockbackParent|GetLauncher) (Variables|KnockbackParent|GetRow_ZAdd))))))

(fn GetAbilityInstigator
  (return (select (Utilities|IsValid (Ability|EffectContext|GetEffectCauser (Ability|GetGrantedbyEffectContext)) :!class /Script/Engine.KismetSystemLibrary) (Variables|/Script/Engine.Actor|GetInstigator (Ability|EffectContext|GetEffectCauser (Ability|GetGrantedbyEffectContext))) (Utilities|Casting|CastToActor (Ability|GetCurrentSourceObject)))))

(fn GetKnockbackOrigin
  (params (OriginActor :type Actor%20Object%20Reference))
  ;[1 1 1] Get the knockback origin. By default this will be the effect causer, this can be overriden to do something custom.
  (return (Transformation|GetActorLocation OriginActor)))

(fn GetTargetLocation
  ;[1 1 1] By Default, knock set the location to launch to be a location away from the launcher based on the knock mags.
  (return (Math|Vector|vector+vector (Transformation|GetActorLocation (Variables|KnockbackParent|GetTarget)) (* (Math|Vector|Normalize (- (Transformation|GetActorLocation (Variables|KnockbackParent|GetTarget)) (Transformation|GetActorLocation (Variables|KnockbackParent|GetLauncher)))) (ScalableFloat|GetValueatLevelwithContextParam (Variables|KnockbackParent|GetLauncher) (Variables|KnockbackParent|GetRow_XYKnockMag))) (Math|Vector|MakeVector 0.0 0.0 (ScalableFloat|GetValueatLevelwithContextParam (Variables|KnockbackParent|GetLauncher) (Variables|KnockbackParent|GetRow_ZKnockMag))) (Math|Vector|MakeVector 0.0 0.0 (ScalableFloat|GetValueatLevelwithContextParam (Variables|KnockbackParent|GetLauncher) (Variables|KnockbackParent|GetRow_ZAdd))))))

(fn LaunchVehicle
  (params (VehicleInterface :type Fort%20Vehicle%20Interface%20Interface) (LaunchVelocity :type Vector%20%28by%20ref%29))
  (bind _returnvalue (Vehicle|GetVehicleActor VehicleInterface :!class /Script/FortniteGame.FortVehicleInterface))
  (Utilities|IsValid _returnvalue
    (:Is%20Valid
      (FortPhysicsSimulationLibrary|SetImpactInstigatorforPhysicsObject _returnvalue (Variables|KnockbackParent|GetLauncher))
      (Audio|BroadcastAbilityImpactatLocation "/Game/Athena/Sounds/Events/Cube/Cube_LaunchPlayer_Cue.Cube_LaunchPlayer_Cue" _returnvalue LaunchVelocity 7.8e+02) ; This is the secret node that launches vehicles. Note it needs the 780 radius to work.
      (Vehicle|WakeUp(Message) _returnvalue))))
