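; Parent custom application requirement for knockback gameplay effects that filters which targets the effect can apply to.
; CanApplyGameplayEffect checks two tag variables against the target's Ability System Component.
; Targets without the building actor tag pass, while those with it pass only if they also carry the building-with-physics tag.
; This blocks the effect on building actors that lack physics.
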(fn CanApplyGameplayEffect
  (params (GameplayEffect :type Gameplay%20Effect%20Object%20Reference) (Spec :type Gameplay%20Effect%20Spec%20Structure%20%28by%20ref%29) (ASC :type Ability%20System%20Component%20Object%20Reference))
  ;[1 1 1] Make sure if the actor has a building tag, that it also has the building physics tag.
  (return (or (not (GameplayTags|HasMatchingGameplayTag ASC (Variables|Default|GetTag_BuildingActor))) (GameplayTags|HasMatchingGameplayTag ASC (Variables|Default|GetTag_BuildingActorWithPhysics)))))
